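# Translucent UI material for genre tile icons in the Discovery browser, tinting a texture parameter according to button state.
# Modular-button functions blend idle, hovered, focused, pressed and disabled states, then their selected variants, into State Color and State Opacity reroutes.
# The icon's red channel is multiplied by a reroute value and an edge-pixel clip computed from normalized UI UVs and texture alpha.
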Material /DiscoveryBrowser/Tiles/Genre/Materials/M_UI_Genre_Icon
graph {
 properties {
  bAutomaticallySetUsageInEditor=false
  bUsedWithStaticMesh=false
  blendMode=BLEND_Translucent
  materialDomain=MD_UI
  usedShadingModels=Default Lit
 }
 nodes [
  n1 = MaterialFunctionCall() {materialFunction=&GetUserInterfaceUV}
  n2 = TextureSample(
   UVs=@n1.Normalized UV
   Tex=TextureObjectParameter() {parameterName=Texture, texture=/Game/UI/DiscoveryShared/Art/T_UI_Discovery_DefaultEmpty}
  ) {samplerSource=SSM_Clamp_WorldGroupSettings}
  Multiply(
   A=@n2.R
   B=Multiply(
    A=MaterialFunctionCall!(UVs=@n1.Normalized UV, Alpha=@n2.A) {materialFunction=/FortUILibrary/MaterialFunctions/Basics/MF_UI_ClipEdgePixel}.Result
    B=NamedRerouteUsage()
   )
  )
  n3 = MaterialFunctionCall!(Idle, Idle Opacity, Hovered, Hovered Opacity, Focused, Focused Opacity, Pressed, Pressed Opacity, Disabled, Disabled Opacity) {materialFunction=/FortUILibrary/Systems/ModularButton/MaterialFunctions/MF_UIKit_ButtonStatesBase_Lerp}
  n4 = MaterialFunctionCall!(
   Base States=@n3.Color
   Base States Opacity=@n3.Color
   Selected_Idle
   Selected_Idle Opacity
   Selected_Hovered
   Selected_Hovered Opacity
   Selected_Focused
   Selected_Focused Opacity
   Selected_Pressed
   Selected_Pressed Opacity
   Selected_Disabled
   Selected_Disabled Opacity
  ) {materialFunction=/FortUILibrary/Systems/ModularButton/MaterialFunctions/MF_UIKit_ButtonStatesSelectable_Lerp}
  NamedRerouteDeclaration(None=@n4.Color) {name=State Color, nodeColor=<rgba 0 0 0 1>}
  NamedRerouteDeclaration(None=@n4.Opacity) {name=State Opacity, nodeColor=<rgba 0 0 0 1>}
  NamedRerouteUsage()
 ]
}
