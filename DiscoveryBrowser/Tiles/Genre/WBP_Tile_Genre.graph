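# Genre tile button for the Discovery browser, showing an icon and wrapping title on a backplate, resized for compact UI.
# List assignment stores the surface tile viewmodel, fetches missing images and passes it to the base tile.
# Clicks send analytics, notify FortActivitySelector, then asynchronously push either a nested surface or View All page, per the viewmodel.
# Image helpers set the icon material's texture parameter and substitute an error texture unless loaded.
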WidgetBlueprint /DiscoveryBrowser/Tiles/Genre/WBP_Tile_Genre
graphs [
 EventGraph {nodes [
   n1 = Event OnListItemObjectSet()
   > n2 = Cast To FortDiscoverSurfaceTileItemVM!(execute, Object=@n1.ListItemObject) -> (then, CastFailed, AsFort Discover Surface Tile Item VM:Fort Discover Surface Tile Item VM Object Reference) {is=title K2Node_DynamicCast}
   > n3 = Set FortDiscoverSurfaceTileItemVM(FortDiscoverSurfaceTileItemVM:Fort Discover Surface Tile Item VM Object Reference=@n2) -> (Output_Get:Fort Discover Surface Tile Item VM Object Reference)
   > n4 = Branch(Condition=HasFetchedImages!(self:Fort Discover Link Entry Tile Item VM Object Reference=@n3) -> (ReturnValue:bool))
   ConfigureDebugUI!(
    execute=[@n5.then, @n4.then]
    self:WBP Discover Tile Base Object Reference=Get WBP_Discover_Tile_Base() -> (WBP_Discover_Tile_Base:WBP Discover Tile Base Object Reference)
    Debug VM:Discover Tile Item Debug View Model Object Reference=Get DebugVM!(self:Discover Tile Item View Model Object Reference=Get FortDiscoverSurfaceTileItemVM() -> (FortDiscoverSurfaceTileItemVM:Fort Discover Surface Tile Item VM Object Reference)) -> (DebugVM:Discover Tile Item Debug View Model Object Reference)
    OwningButton:Common Button Base Object Reference=Self-Reference()
   ) -> (then)
   > SetTileItemViewModel!(
    execute
    self:WBP Discover Tile Base Object Reference=Get WBP_Discover_Tile_Base()
    TileItemViewModel:Discover Tile Item View Model Object Reference=@n3
   ) -> (then)
   n5 = FetchImages(execute=@n4.else, self:Fort Discover Link Entry Tile Item VM Object Reference=@n3)
   Event Construct()
   > Parent: Construct()
   Event OnInitialized()
   > SetCompactUIStyle() {is=title K2Node_CallFunction}
   n6 = Event PreConstruct()
   > Parent: PreConstruct(IsDesignTime=@n6.IsDesignTime)
   > Set Tile Size!(
    execute
    self:WBP Discover Tile Base Object Reference=Get WBP_Discover_Tile_Base()
    InSize:Vector 2D Structure=Get TileSize() -> (TileSize:Vector 2D Structure)
   ) -> (then) {is=title K2Node_CallFunction}
   > SetMaterial!(
    execute
    self:WBP UIKit Block Image Object Reference=Get Block_Icon() -> (Block_Icon:WBP UIKit Block Image Object Reference)
    InMaterial:"Material Instance Object Reference (by ref)"=Get IconMaterialInstance() -> (IconMaterialInstance:Material Instance Object Reference)
   ) -> (then)
   n7 = Get WBP_UIKit_ContextMenuAnchorButton!(self:WBP Discover Tile Base Object Reference=Get WBP_Discover_Tile_Base()) -> (WBP_UIKit_ContextMenuAnchorButton:WBP UIKit Context Menu Anchor Button Object Reference)
   > SetVisibility(
    self=[@n7, Get Anchor!(self:UIKit Context Menu Anchor Object Reference=@n7) -> (Anchor:Menu Anchor Object Reference)]
    InVisibility=Collapsed
   )
   > RegisterModularBlocks()
   > UpdateTileVisual()
   Event On Clicked()
   > Parent: On Clicked() {is=title K2Node_CallParentFunction}
   n8 = Get FortDiscoverSurfaceTileItemVM()
   > SendFlowRiderClickTileEvent!(
    execute
    self:WBP Discover Tile Base Object Reference=Get WBP_Discover_Tile_Base()
    TileWidget:User Widget Object Reference=Self-Reference()
    TileType:EActivityItemType Enum=PanelTile
    TileUniqueIdentifier:String="To String (Name)"(InName=GetReferenceSurfaceName!(self:Fort Discover Surface Tile Item VM Object Reference=@n8) -> (ReturnValue:Name=None))
   ) -> (then)
   n9 = Get FortDiscoverSurfaceTileItemVM()
   > HandleCategoryTileClicked!(
    execute
    self:Fort Activity Selector Object Reference=Get FortActivitySelector() -> (ReturnValue:Fort Activity Selector Object Reference) {is=title K2Node_GetSubsystem}
    FortPC:Fort Player Controller Object Reference=GetFirstLocalFortPlayerController!(bIncludeNonSpectators:bool=true, bIncludeSpectators:bool=false) -> (ReturnValue:Fort Player Controller Object Reference)
    InTileItemVM:Fort Discover Surface Tile Item VM Object Reference=@n9
   ) -> (then)
   > n10 = Branch(Condition=GetUseNestedDiscoverView!(self:Fort Discover Surface Tile Item VM Object Reference=@n9) -> (ReturnValue:bool))
   > n11 = PushPrimaryContentClassAsync!(
    execute:Exec
    WidgetClass:Common Activatable Widget Class Reference=/DiscoveryBrowser/WBP_Discover_NestedSurface.WBP_Discover_NestedSurface_C
   ) -> (then:Exec, OnPrimaryContentPushed:Exec, ActivatableWidget:Common Activatable Widget Object Reference) {is=title K2Node_AsyncAction}
   n12 = Cast To WBP_Discover_NestedSurface!(execute=@n11.OnPrimaryContentPushed, Object=@n11) -> (then, CastFailed, AsWBP Discover Nested Surface:WBP Discover Nested Surface Object Reference) {is=title K2Node_DynamicCast}
   > SetFortDiscoverSurfaceTileItemVM!(
    execute
    self:WBP Discover Nested Surface Object Reference=@n12
    Viewmodel:Fort Discover Surface Tile Item VM Object Reference=Get FortDiscoverSurfaceTileItemVM()
   ) -> (then)
   > On Disover Surface Tile Item VM Set(self:WBP Discover Nested Surface Object Reference=@n12)
   > SetSurfaceName!(
    execute
    self:Discover Surface View Model Object Reference=Get FortDiscoverSurfaceViewModel!(self:WBP Discover Nested Surface Object Reference=@n12) -> (FortDiscoverSurfaceViewModel:Discover Surface View Model Object Reference)
    InSurfaceName:Name=GetReferenceSurfaceName(self=Get FortDiscoverSurfaceTileItemVM()) -> (ReturnValue=None)
   ) -> (then)
   n13 = PushPrimaryContentClassAsync(execute=@n10.else, WidgetClass=/DiscoveryBrowser/WBP_ViewAllPage.WBP_ViewAllPage_C)
   n14 = Cast To WBP_ViewAllPage!(execute=@n13.OnPrimaryContentPushed, Object=@n13) -> (then, CastFailed, AsWBP View All Page:WBP View All Page Object Reference) {is=title K2Node_DynamicCast}
   > SetupSurface!(
    execute
    self:WBP View All Page Object Reference=@n14
    Surface Name:Name=GetReferenceSurfaceName(self=@n9) -> (ReturnValue=None)
    Panel Name:Name=GetReferencePanelName!(self:Fort Discover Surface Tile Item VM Object Reference=@n9) -> (ReturnValue:Name=None)
   ) -> (then)
   > ActivateWidget(self=@n13)
   n15 = Event SetNestedButtonGroupSelectionVM!() -> (OutputDelegate, then, InVM:Fort Nested Button Group Selection VM Object Reference) {is=title K2Node_Event}
   > Set FortNestedButtonGroupSelectionVM(FortNestedButtonGroupSelectionVM:Fort Nested Button Group Selection VM Object Reference=@n15.InVM) -> (Output_Get:Fort Nested Button Group Selection VM Object Reference)
  ]}
 GetImageTexture {nodes [
   n1 = GetImageTexture!() -> (then, ImageLoadingState:EActivityImageLoadingState Enum=Unset, ImageTexture:Texture Object Reference) {is=title K2Node_FunctionEntry}
   > n2 = Switch on EActivityImageLoadingState!(execute, Selection:EActivityImageLoadingState Enum=@n1) -> (Unset:Exec, Loading:Exec, Loaded:Exec, Error:Exec) {is=title K2Node_SwitchEnum}
   Return Node!(execute=[@n2.Error, @n2.Loading, @n2.Unset], Texture:Texture Object Reference=/FortUILibrary/Textures/Icons/T_UI_Icon_UIKit_ImageError)
   Return Node(execute=@n2.Loaded, Texture=@n1.ImageTexture)
  ]}
 GetNestedButtonGroupSelectionVM {nodes [
   GetNestedButtonGroupSelectionVM() {is=title K2Node_FunctionEntry}
   > Return Node!(
    execute
    ReturnValue:Fort Nested Button Group Selection VM Object Reference=Get FortNestedButtonGroupSelectionVM() -> (FortNestedButtonGroupSelectionVM:Fort Nested Button Group Selection VM Object Reference)
   )
  ]}
 RegisterModularBlocks {nodes [
   RegisterModularBlocks() {is=title K2Node_FunctionEntry}
   > RegisterModularBlock!(
    execute
    self:UIKit Modular Button Object Reference
    Block:UIKit Block Object Reference=Get Block_Text() -> (Block_Text:WBP UIKit Block Text Object Reference)
   ) -> (then)
   > RegisterModularBlock(Block=Get Block_Icon())
  ]}
 SetCompactUIStyle {nodes [
   SetCompactUIStyle() {is=title K2Node_FunctionEntry}
   > Branch(Condition=IsUsingCompactUI())
   > Set TileSize(TileSize:Vector 2D Structure=(X=312,Y=96)) -> (Output_Get:Vector 2D Structure)
   > SetCornerRadius!(
    execute
    self:WBP UIKit Backplate Base Object Reference=Get Backplate() -> (Backplate:WBP UIKit Backplate Object Reference)
    Corner Radius:E_UI_BackplateCornerRadius Enum=NewEnumerator6
   ) -> (then)
  ]}
 SetImageTexture {nodes [
   n1 = SetImageTexture!() -> (then, IconTexture:Texture Object Reference) {is=title K2Node_FunctionEntry}
   > SetTextureParameterValue(
    self=Get ImageMID!(self:WBP UIKit Block Image Object Reference=Get Block_Icon()) -> (ImageMID:Material Instance Dynamic Object Reference)
    ParameterName=Texture
    Value=@n1
   )
  ]}
 UpdateTileVisual {nodes [
   UpdateTileVisual() {is=title K2Node_FunctionEntry}
   > SetFont!(
    execute
    self:WBP UIKit Block Text Object Reference=Get Block_Text()
    Font:"Slate Font Info Structure (by ref)"=Get LTitleTextFont() -> (LTitleTextFont:Slate Font Info Structure)
   ) -> (then)
   > SetTransformPolicy!(execute, self:WBP UIKit Block Text Object Reference=Get Block_Text(), Transform Policy:ETextTransformPolicy Enum=None) -> (then)
   > SetTextOverflowPolicy!(execute, self:WBP UIKit Block Text Object Reference=Get Block_Text(), OverflowPolicy:ETextOverflowPolicy Enum=MultilineEllipsis) -> (then)
   > SetJustification!(execute, self:WBP UIKit Block Text Object Reference=Get Block_Text(), Justification:ETextJustify Enum=Left) -> (then)
   n1 = Get TextBlock!(self:WBP UIKit Block Text Object Reference=Get Block_Text()) -> (TextBlock:Common Text Object Reference)
   > SetScrollingEnabled!(execute, self:Common Text Object Reference=@n1, bInIsScrollingEnabled:bool=false) -> (then)
   > Set Clipping!(execute, Clipping:EWidgetClipping Enum=Inherit, self:Widget Object Reference=@n1) -> (then, Output_Get:EWidgetClipping Enum=Inherit)
   > Set AutoWrapText!(execute, AutoWrapText:bool=true, self:Text Layout Widget Object Reference=@n1) -> (then, Output_Get:bool)
  ]}
]
variables [
 IconMaterialInstance {type=MaterialInstance, default=/DiscoveryBrowser/Tiles/Genre/Materials/MI_UI_Genre_Icon}
 TileSize {type=Vector2D, default=<256 72>}
]
widgets {
 info {
  inheritedWidgetCount=0
  namedSlotCount=1
  parentClass=/DiscoveryBrowser/Tiles/WBP_Discover_Tile_Button.WBP_Discover_Tile_Button_C
  rootWidgetClass=null
  widgetCount=4
 }
 widgets [
  GridPanel_52 {widgetClassPath=&GridPanel, widget=GridPanel_52, properties {columnFill [0,1,0], rowFill [1]}}
  Backplate {
   widgetClassPath=&WBP_UIKit_Backplate
   bIsVariable=true
   parent=GridPanel_52
   widget=Backplate
   properties {brightness=Medium}
   slot {class=GridSlot, properties {columnSpan=2}}
  }
  Block_Icon {
   widgetClassPath=/FortUILibrary/Systems/ModularButton/Blocks/WBP_UIKit_Block_Image
   bIsVariable=true
   parent=GridPanel_52
   widget=Block_Icon
   properties {
    pixelSnapping=Disabled
    transitionTiming {
     deselectingFocused=0
     deselectingIdle=0
     disabling=0
     enabling=0
     focusing=0.2
     hovering=0.2
     pressing=0
     releasing=0
     selecting=0
     unfocusing=0.2
     unhovering=0.2
    }
   }
   slot {class=GridSlot, properties {horizontalAlignment=HAlign_Left, layer=1, verticalAlignment=VAlign_Center, padding=<ltrb 16 0 0 0>}}
  }
  Block_Text {
   widgetClassPath=/FortUILibrary/Systems/ModularButton/Blocks/WBP_UIKit_Block_Text
   bIsVariable=true
   parent=GridPanel_52
   widget=Block_Text
   properties {transitionTiming {
     deselectingFocused=0
     deselectingIdle=0
     disabling=0
     enabling=0
     focusing=0.2
     hovering=0.2
     pressing=0
     releasing=0
     selecting=0
     unfocusing=0.2
     unhovering=0.2
    }}
   slot {class=GridSlot, properties {column=1, layer=1, verticalAlignment=VAlign_Center, padding=<ltrb 16 0 16 0>}}
  }
 ]
}
